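# Niagara system for the swing trail of the LatteStir pickaxe, drawn as animation-trail ribbons by two emitters, Main_Trail and Tip_Trail.
# Both use shared FN anim-trail modules for spawn and update, with Tip_Trail reading its own pair of bone or socket indices.
# Time-of-day color curves tint both trails, and User.SwingSectionIndex selects system-age alpha curves that time each fade per swing section.
# Main_Trail also computes a dark color with time-of-day-scaled opacity and scales alpha by SystemDissolve.
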NiagaraSystem /BRCosmetics/Weapons/FORT_Melee/Pickaxe_LatteStir/FX/NS_Pickaxe_LatteStir_Trail
dependencies {
 usedDataInterfaces [&NiagaraDataInterfaceColorCurve, &NiagaraDataInterfaceCurve, &NiagaraDataInterfaceSocketReader]
 usedDynamicInputs []
 usedModules [
  &self:SystemSpawnScript.SystemScriptSource.SystemScriptGraph.NiagaraNodeAssignment_5.SetVariables_#1
  &SystemState
  &self:SystemSpawnScript.SystemScriptSource.SystemScriptGraph.NiagaraNodeAssignment_9.SetVariables_#2
  &self:SystemSpawnScript.SystemScriptSource.SystemScriptGraph.NiagaraNodeAssignment_10.SetVariables_#3
  &self:SystemSpawnScript.SystemScriptSource.SystemScriptGraph.NiagaraNodeAssignment_11.SetVariables_#4
 ]
}
emitters {
 Main_Trail {
  emitterSpawnScript [&FN_RibbonIDOffsetUtility()]
  emitterUpdateScript [
   &EmitterState()
   &FN_SpawnAnimTrail()
  ]
  particleSpawnScript [
   &FN_InitializeAnimTrailBoneIndecies()
   &FN_InitializeAnimTrail(Lifetime=0.3)
  ]
  particleUpdateScript [
   &FN_UpdateAnimTrail() {module=FN_UpdateAnimTrail001}
   &ScaleColor(
    Scale Alpha=$System.SystemDissolve
    Scale RGB=&MakeVectorFromLinearColorRGB(Color=&ColorFromCurve(
      ColorCurve=DataInterface() {propertyValues {
        AlphaCurve {keys [<at 0.875 1 RCIM_Cubic>]}
        BlueCurve {keys [
          <at 0 0.45>
          {arriveTangent=1.0726, interpMode=RCIM_Cubic, leaveTangent=1.0726, time=0.0833, value=0.5776}
          <at 0.125 0.5841>
          {arriveTangent=0.0723, interpMode=RCIM_Cubic, leaveTangent=0.0723, time=0.4167, value=0.6172}
          <at 0.5833 0.6172>
          <at 0.75 0.5841>
          {arriveTangent=-0.039, interpMode=RCIM_Cubic, leaveTangent=-0.039, time=0.875, value=0.5128}
          <at 0.9167 0.5776>
          <at 1 0.45>
         ]}
        GreenCurve {keys [
          <at 0 0.214>
          {arriveTangent=2.4562, interpMode=RCIM_Cubic, leaveTangent=2.4562, time=0.0833, value=0.4328}
          <at 0.125 0.521>
          {arriveTangent=0.1094, interpMode=RCIM_Cubic, leaveTangent=0.1094, time=0.4167, value=0.5711}
          <at 0.5833 0.5711>
          <at 0.75 0.521>
          {arriveTangent=-0.5293, interpMode=RCIM_Cubic, leaveTangent=-0.5293, time=0.875, value=0.5341}
          <at 0.9167 0.4328>
          <at 1 0.214>
         ]}
        RedCurve {keys [
          <at 0 0.1238>
          {arriveTangent=2.3961, interpMode=RCIM_Cubic, leaveTangent=2.3961, time=0.0833, value=0.2599}
          <at 0.125 0.4233>
          {arriveTangent=0.1607, interpMode=RCIM_Cubic, leaveTangent=0.1607, time=0.4167, value=0.4969}
          <at 0.5833 0.4969>
          <at 0.75 0.4233>
          {arriveTangent=-0.9801, interpMode=RCIM_Cubic, leaveTangent=-0.9801, time=0.875, value=0.586}
          <at 0.9167 0.2599>
          <at 1 0.1238>
         ]}
       }}
      CurveIndex=Linked() {linkedVariable {name=NPC.FortniteNPC.RemappedTimeOfDay_0-1, type=NiagaraFloat}}
     ))
    ScaleA=true
    ScaleRGB=true
   )
   &FN_Sprite_Attributes(
    Color=&MakeLinearColorFromVectorAndFloat(
     "Vector (RGB)"=&MakeVectorFromLinearColorRGB(Color=<rgba 0.0242 0.0369 0.0497 1>)
     "Float (Alpha)"=&Multiply_Float(
      A=&Multiply_Float(
       A=&FloatFromCurve(
        FloatCurve=DataInterface() {propertyValues {Curve {keys [<at 0 1 RCIM_Cubic>, {arriveTangent=-8.8973, interpMode=RCIM_Cubic, tangentMode=RCTM_Break, time=0.3, value=0}]}}}
        Scale Curve=&FNTimeOfDayScalar(Night Scale=0.45, Morning Scale=0.5, Day Scale=0.6, Evening Scale=0.5)
       )
       B=&MergeHelper_FloatSwitch(
        0=&FloatFromCurve(FloatCurve=DataInterface() {propertyValues {Curve {keys [<at 0 1 RCIM_Cubic>,<at 0.175 0>]}}}, CurveIndex=$Engine.System.Age)
        1=&FloatFromCurve(FloatCurve=DataInterface() {propertyValues {Curve {keys [<at 0 1 RCIM_Cubic>,<at 0.17 0>]}}}, CurveIndex=$Engine.System.Age)
        2=&FloatFromCurve(FloatCurve=DataInterface() {propertyValues {Curve {keys [<at 0 1 RCIM_Cubic>,<at 0.145 0>]}}}, CurveIndex=$Engine.System.Age)
        3=&FloatFromCurve(FloatCurve=DataInterface() {propertyValues {Curve {keys [<at 0.03 1 RCIM_Cubic>,<at 0.17 0 RCIM_Cubic>]}}}, CurveIndex=$Engine.System.Age)
        Index=$User.SwingSectionIndex
        Num=::NewEnumerator2
       )
      )
      B=3
     )
    )
    Create Color Attribute=true
    Write to Module Namespace=true
   )
  ]
  rendererClasses [&NiagaraRibbonRendererProperties]
  renderers [
   {rendererClass=&NiagaraRibbonRendererProperties}
   {rendererClass=&NiagaraRibbonRendererProperties, rendererIndex=1}
  ]
 }
 Tip_Trail {
  emitterSpawnScript [&FN_RibbonIDOffsetUtility()]
  emitterUpdateScript [
   &EmitterState()
   &FN_SpawnAnimTrail()
  ]
  particleSpawnScript [
   &FN_InitializeAnimTrailBoneIndecies(First Bone or Socket Index=1, Second Bone or Socket Index=2) {module=FN_InitializeAnimTrailBoneIndecies001}
   &FN_InitializeAnimTrail(Color=<rgba 2 2 2 1>, Lifetime=0.3)
  ]
  particleUpdateScript [
   &FN_UpdateAnimTrail() {module=FN_UpdateAnimTrail001}
   &ScaleColor(
    Scale Alpha=&MergeHelper_FloatSwitch(
     0=&FloatFromCurve(FloatCurve=DataInterface() {propertyValues {Curve {keys [<at 0.1 1>,<at 0.17 0>]}}}, CurveIndex=$System.Age)
     1=&FloatFromCurve(FloatCurve=DataInterface() {propertyValues {Curve {keys [<at 0.12 1>,<at 0.195 0>]}}}, CurveIndex=$System.Age)
     2=&FloatFromCurve(FloatCurve=DataInterface() {propertyValues {Curve {keys [<at 0.1 1>,<at 0.195 0>]}}}, CurveIndex=$System.Age)
     3=&FloatFromCurve(FloatCurve=DataInterface() {propertyValues {Curve {keys [<at 0.14 1>,<at 0.195 0>]}}}, CurveIndex=$System.Age)
     Index=$User.SwingSectionIndex
     Num=::NewEnumerator2
    )
    Scale RGB=&MakeVectorFromLinearColorRGB(Color=&ColorFromCurve(
      ColorCurve=DataInterface() {propertyValues {
        AlphaCurve {keys [<at 0.875 1 RCIM_Cubic>]}
        BlueCurve {keys [<at 0 0.6172>,<at 0.0833 0.6939 RCIM_Cubic>,<at 0.125 0.6172>,<at 0.4167 0.6172 RCIM_Cubic>,<at 0.5833 0.6172>,<at 0.75 0.617>,<at 0.875 0.5738 RCIM_Cubic>,<at 0.9167 0.617>,<at 1 0.6172>]}
        GreenCurve {keys [
          <at 0 0.3501>
          {arriveTangent=1.9102, interpMode=RCIM_Cubic, leaveTangent=1.9102, time=0.0833, value=0.515}
          <at 0.125 0.5888>
          {arriveTangent=-0.0386, interpMode=RCIM_Cubic, leaveTangent=-0.0386, time=0.4167, value=0.5711}
          <at 0.5833 0.5711>
          <at 0.75 0.5504>
          {arriveTangent=-0.6126, interpMode=RCIM_Cubic, leaveTangent=-0.6126, time=0.875, value=0.5864}
          <at 0.9167 0.4483>
          <at 1 0.3501>
         ]}
        RedCurve {keys [
          <at 0 0.2469>
          {arriveTangent=2.3701, interpMode=RCIM_Cubic, leaveTangent=2.3701, time=0.0833, value=0.2775}
          <at 0.125 0.5431>
          {arriveTangent=-0.1008, interpMode=RCIM_Cubic, leaveTangent=-0.1008, time=0.4167, value=0.4969}
          <at 0.5833 0.4969>
          <at 0.75 0.4471>
          {arriveTangent=-1.2019, interpMode=RCIM_Cubic, leaveTangent=-1.2019, time=0.875, value=0.617}
          <at 0.9167 0.2468>
          <at 1 0.2469>
         ]}
       }}
      CurveIndex=Linked() {linkedVariable {name=NPC.FortniteNPC.RemappedTimeOfDay_0-1, type=NiagaraFloat}}
     ))
    ScaleA=true
    ScaleRGB=true
   )
  ]
  rendererClasses [&NiagaraRibbonRendererProperties]
  renderers [{rendererClass=&NiagaraRibbonRendererProperties}]
 }
}
summary {
 systemName=NS_Pickaxe_LatteStir_Trail
 emitters [
  {emitterName=Main_Trail, rendererClasses [&NiagaraRibbonRendererProperties]}
  {emitterName=Tip_Trail, rendererClasses [&NiagaraRibbonRendererProperties]}
 ]
 userVariables [
  User.RibbonIDOffset {type=NiagaraInt32}
  User.SwingSectionIndex {type=NiagaraFloat, defaultValue=0, description=""}
  User.SwingSectionIndex {type=NiagaraInt32}
 ]
}
