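# Physics asset for the LatteStir pickaxe, a two-handed melee cosmetic, that lets a dangling chain on the weapon move physically.
# A capsule on the handle bone anchors six small, identically sized capsule bodies on bones dyn_chain_a through dyn_chain_f.
# Constraints link each body to the previous one in a single serial chain, pairing wide swing limits with very tight twist limits so it bends freely without spinning.
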PhysicsAsset /BRCosmetics/Weapons/FORT_Melee/Pickaxe_LatteStir/Meshes/LatteStir_Axe_Physics
bodies {
 Melee_TwoHanded_Handle [{
   shapeType=Capsule
   length=4.6796
   radius=2.1877
   shapeName=Melee_TwoHanded_Handle_capsule
   center=<-0.0035 0.1769 -15.4216>
   rotation=<rot -0.1474 -90 0.0178>
  }]
 dyn_chain_a [{shapeType=Capsule, length=1.5, radius=1.1, shapeName=dyn_chain_a_capsule, center=<0.6 0 0>, rotation=<rot 0 90 -90>}]
 dyn_chain_b [{shapeType=Capsule, length=1.5, radius=1.1, shapeName=dyn_chain_b_capsule, center=<0.6 0 0>, rotation=<rot 0 90 -90>}]
 dyn_chain_c [{shapeType=Capsule, length=1.5, radius=1.1, shapeName=dyn_chain_c_capsule, center=<0.8 0 0>, rotation=<rot 0 90 -90>}]
 dyn_chain_d [{shapeType=Capsule, length=1.5, radius=1.1, shapeName=dyn_chain_d_capsule, center=<0.8 0 0>, rotation=<rot 0 90 -90>}]
 dyn_chain_e [{shapeType=Capsule, length=1.5, radius=1.1, shapeName=dyn_chain_e_capsule, center=<0.8 0 0>, rotation=<rot 0 90 -90>}]
 dyn_chain_f [{shapeType=Capsule, length=1.5, radius=1.1, shapeName=dyn_chain_f_capsule, center=<0.8 0 0>, rotation=<rot 0 90 -90>}]
}
constraints [
 {bone1Name=dyn_chain_f, bone2Name=dyn_chain_e, swing1LimitDegrees=90, swing2LimitDegrees=85, twistLimitDegrees=4}
 {bone1Name=dyn_chain_b, bone2Name=dyn_chain_a, swing1LimitDegrees=100, swing2LimitDegrees=85, twistLimitDegrees=4}
 {bone1Name=dyn_chain_c, bone2Name=dyn_chain_b, swing1LimitDegrees=80, swing2LimitDegrees=85, twistLimitDegrees=4}
 {bone1Name=dyn_chain_d, bone2Name=dyn_chain_c, swing1LimitDegrees=80, swing2LimitDegrees=85, twistLimitDegrees=4}
 {bone1Name=dyn_chain_e, bone2Name=dyn_chain_d, swing1LimitDegrees=85, swing2LimitDegrees=85, twistLimitDegrees=4}
 {bone1Name=dyn_chain_a, bone2Name=Melee_TwoHanded_Handle, swing1LimitDegrees=100, swing2LimitDegrees=90, twistLimitDegrees=4}
]
